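# UI-domain alpha-composite material for the frame of the MurkyVole Gauntlets interface.
# A grayscale frame texture masks both the translucency and a vertical color gradient, squared for falloff, between TopColor and BottomColor parameters.
# UVs drift with the CameraVelocity parameter, scaled down and clamped, so the frame moves less than other elements.
# A procedural Brackets mask, box outline minus box fill, is declared with no visible consumer.
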Material /MurkyVoleGameplay/Gameplay/GauntletGuns/UI/V2/M_UI_MurkyVole_Gauntlets_Frame
graph {
 properties {bAutomaticallySetUsageInEditor=false, blendMode=BLEND_AlphaComposite, materialDomain=MD_UI, usedShadingModels=Default Lit}
 nodes [
  n1 = TextureSample(UVs=NamedRerouteUsage()) {samplerType=SAMPLERTYPE_Grayscale, texture=&/T_MurkyVole_Frame}
  LinearInterpolate(B=ScalarParameter() {defaultValue=0.15, parameterName=Translucency}, Alpha=@n1.R)
  n2 = Multiply(A=ComponentMask(None=NamedRerouteUsage()) {g=true}) {constB=2}
  LinearInterpolate(
   B=LinearInterpolate(
    A=VectorParameter() {parameterName=TopColor, defaultValue=<rgba 0.0056 0.8879 0.1812 1>}.RGB
    B=VectorParameter() {parameterName=BottomColor, defaultValue=<rgba 0.9911 0.9911 0.9911 1>}.RGB
    Alpha=Multiply(A=@n2, B=@n2)
   )
   Alpha=@n1.R
  )
  n3 = Divide(A=ComponentMask(None=VectorParameter() {parameterName=CameraVelocity}.RGB) {g=true, r=true}, B=Constant2Vector() {g=0.84, r=9.6}) {desc="Decreases movement\r\nof element relative to others"}
  n4 = Constant2Vector() {desc=Max Drift, g=0.09, r=0.1}
  n5 = Multiply(A=@n4) {constB=-1}
  NamedRerouteDeclaration(None=Add(
    A=MaterialFunctionCall() {materialFunction=&GetUserInterfaceUV}.Normalized UV
    B=Clamp(None=@n3, Min=@n5, Max=@n4)
   )) {name=UVs, nodeColor=<rgba 0 0 0 1>}
  n6 = Constant2Vector() {g=0.81, r=0.78}
  n7 = Constant() {r=-1}
  n8 = Constant() {r=0.006}
  n9 = Constant() {r=0.001}
  n10 = Constant2Vector() {g=1, r=0.685}
  NamedRerouteDeclaration(None=Subtract(
    A=MaterialFunctionCall(Size=@n6, CornerRoundness=@n7, OutlineThickness=@n8, OutlineAlignment=@n7, GlowMax=@n9) {materialFunction=&MF_UI_Box}.Outline
    B=MaterialFunctionCall(Size=@n10) {materialFunction=&MF_UI_Box}.Fill
   )) {name=Brackets, nodeColor=<rgba 0 0.7891 1 1>}
 ]
}
