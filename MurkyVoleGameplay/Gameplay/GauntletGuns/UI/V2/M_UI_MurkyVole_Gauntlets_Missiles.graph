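# UI material that draws the gauntlet guns' remaining-missile display as a grid of icons.
# ClusterSize and ClusterCount tile inner and outer icon textures, and each cell's floored-UV index, offset by RocketsRemaining, weights the outer layer to show ammo.
# CameraVelocity shifts the scaled UVs by a damped, clamped amount for HUD drift.
# An MF_UI_Box mask clips the result to the grid.
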Material /MurkyVoleGameplay/Gameplay/GauntletGuns/UI/V2/M_UI_MurkyVole_Gauntlets_Missiles
graph {
 properties {bAutomaticallySetUsageInEditor=false, blendMode=BLEND_AlphaComposite, materialDomain=MD_UI, usedShadingModels=Default Lit}
 nodes [
  n1 = Floor(None=AppendVector(A=ScalarParameter() {defaultValue=2, parameterName=ClusterSize}, B=ScalarParameter() {defaultValue=4, parameterName=ClusterCount}))
  n2 = Multiply(A=Add(A=NamedRerouteUsage(), B=NamedRerouteUsage()) {desc=Offset for mask}, B=@n1)
  n3 = TextureSample(UVs=@n2) {
   samplerSource=SSM_Wrap_WorldGroupSettings
   samplerType=SAMPLERTYPE_Grayscale
   texture=/MurkyVoleGameplay/Gameplay/GauntletGuns/UI/Textures/T_Missile_IconInner
  }
  n4 = SmoothStep(Value=TextureSample(UVs=@n2) {samplerType=SAMPLERTYPE_Grayscale, texture=/MurkyVoleGameplay/Gameplay/GauntletGuns/UI/Textures/T_Missile_IconOuter}.R) {constMax=0.5, constMin=0.15}
  n5 = Floor(None=@n2)
  n6 = ComponentMask(None=@n5) {r=true}
  n7 = ComponentMask(None=@n1) {r=true}
  n8 = Multiply(A=ComponentMask(None=@n5) {g=true}, B=@n7)
  n9 = ComponentMask(None=@n1) {g=true}
  n10 = Subtract(A=Add(A=Add(A=@n6, B=@n8)), B=Subtract(A=Multiply(A=@n7, B=@n9), B=ScalarParameter() {defaultValue=16, parameterName=RocketsRemaining}))
  Min(A=Add(A=SmoothStep(Value=@n3.R) {constMax=0.3, constMin=0.1}, B=Multiply(A=@n4, B=@n10)), B=NamedRerouteUsage())
  n11 = MaterialFunctionCall() {materialFunction=&GetUserInterfaceUV}
  n12 = Divide(B=ScalarParameter() {defaultValue=0.8, parameterName=ImageScalar}) {constB=0.5}
  n13 = Divide(A=ComponentMask(None=VectorParameter() {parameterName=CameraVelocity}.RGB) {g=true, r=true}, B=Constant2Vector() {g=0.84, r=9.6}) {desc="Decreases movement\r\nof element relative to others"}
  n14 = Constant2Vector() {desc=Max Drift, g=0.15, r=0.2}
  n15 = Multiply(A=@n14) {constB=-1}
  NamedRerouteDeclaration(None=Add(A=Multiply(A=@n11.Normalized UV, B=@n12) {constB=0.5}, B=Clamp(None=@n13, Min=@n15, Max=@n14))) {name=UVs, nodeColor=<rgba 0 0 0 1>}
  n16 = MaterialFunctionCall(UVs=Add(A=NamedRerouteUsage(), B=NamedRerouteUsage()), Size=Constant2Vector() {g=1, r=1}) {materialFunction=&MF_UI_Box}
  NamedRerouteDeclaration(None=Max(A=Multiply(A=@n16.Fill) {constB=-1}, B=@n16.Fill)) {name=Mask, nodeColor=<rgba 0 0 0 1>}
  n17 = ScalarParameter() {defaultValue=0.8, parameterName=ImageScalar}
  n18 = Multiply(A=Divide(B=@n17) {constB=0.5}, B=Divide(A=OneMinus(None=@n17)) {constB=-2}) {constB=0.167}
  NamedRerouteDeclaration(None=AppendVector(A=@n18, B=@n18)) {name=CenterOffset, nodeColor=<rgba 0 0 0 1>}
 ]
}
