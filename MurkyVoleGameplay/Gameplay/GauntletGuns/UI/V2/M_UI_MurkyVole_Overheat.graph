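# Additive UI material that draws the gauntlet gun overheat meter as a vertical, segmented bar driven by the Overheat scalar.
# A step on texture V sets the fill level, Fmod pips clipped by an MF_UI_Box shape form segments, and a smoothstepped frame texture gives the bar mask.
# Color blends from Fill toward OverheatFill at high heat.
# The Overheated? parameter switches in a time-based sine flash, and vertex alpha scales the colored result.
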Material /MurkyVoleGameplay/Gameplay/GauntletGuns/UI/V2/M_UI_MurkyVole_Overheat
graph {
 properties {bAutomaticallySetUsageInEditor=false, blendMode=BLEND_Additive, materialDomain=MD_UI, usedShadingModels=Default Lit}
 nodes [
  n1 = ScalarParameter() {defaultValue=1, parameterName=Overheat}
  n2 = Add(A=@n1) {constB=0.25}
  n3 = VectorParameter() {parameterName=Fill, defaultValue=<rgba 1 1 1 1>}
  n4 = VectorParameter() {parameterName=OverheatFill, defaultValue=<rgba 1 0 0.0284 1>}
  n5 = If(A=@n1, "A > B"=LinearInterpolate(A=@n3.RGB, B=@n4.RGB, Alpha=Saturate(None=@n2)) {constB=0.5}, "A < B"=@n3.RGB) {constB=0.55}
  n6 = NamedRerouteUsage()
  n7 = NamedRerouteUsage()
  n8 = Multiply(A=@n6, B=@n7)
  n9 = Subtract(A=@n1) {constB=0.75}
  n10 = If(A=@n1, "A > B"=LinearInterpolate(A=@n3.RGB, B=@n4.RGB, Alpha=Multiply(A=@n9) {constB=4}), "A < B"=@n3.RGB) {constB=0.7}
  n11 = Multiply(A=ScalarParameter() {defaultValue=1, parameterName=Overheat})
  n12 = NamedRerouteUsage()
  n13 = Multiply(A=@n12, B=@n11) {constB=0.4}
  Multiply(A=Add(A=Multiply(A=@n5, B=@n8), B=Multiply(A=@n10, B=@n13)), B=VertexColor().A)
  Multiply(A=Add(A=Multiply(A=@n6, B=@n7), B=@n12), B=@n11)
  n14 = NamedRerouteUsage()
  n15 = OneMinus(None=ComponentMask(None=TextureCoordinate()) {g=true})
  n16 = Step(Y=@n15, X=LinearInterpolate(Alpha=ScalarParameter() {defaultValue=1, parameterName=Overheat}) {constA=0.075, constB=0.926}) {constX=0}
  n17 = ScalarParameter() {defaultValue=33, parameterName=Segment Amount}
  n18 = Step(Y=Fmod(A=@n15, B=Divide(B=@n17) {constB=30})) {constX=0.012}
  n19 = MaterialFunctionCall(Size=Constant2Vector() {g=0.85, r=0.275}) {materialFunction=&MF_UI_Box}
  Multiply(
   A=NamedRerouteDeclaration(None=Multiply(A=@n14, B=@n16)) {name=Progress Fill, nodeColor=<rgba 0.3125 1 0 1>}
   B=NamedRerouteDeclaration(None=Multiply(A=@n18, B=@n19.Fill) {constB=0.5}) {name=PipAlpha, nodeColor=<rgba 1 0 0.6328 1>}
  )
  n20 = Multiply(A=Time()) {constB=2.5}
  n21 = ScalarParameter() {parameterName=Overheated?}
  NamedRerouteDeclaration(None=LinearInterpolate(B=Sine(None=@n20), Alpha=Floor(None=@n21)) {constA=1}) {name=OverheatFlash, nodeColor=<rgba 0 0.5781 1 1>}
  n22 = Add(A=TextureCoordinate(), B=Constant2Vector() {r=-0.05})
  NamedRerouteDeclaration(None=SmoothStep(Value=TextureSample(UVs=@n22) {samplerType=SAMPLERTYPE_LinearGrayscale, texture=&/T_WM_OH_MeterFrame}.R) {constMax=0.5, constMin=0.46}) {name=BarTexture, nodeColor=<rgba 0 1 0.3203 1>}
 ]
}
